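# Unlit translucent Niagara sprite and mesh particle material that draws a flickering neon disco floor effect for the Sparkle Chop pickaxe.
# Tiled UV cells plus stepped time feed a sine-warped Perlin lookup, giving each tile a changing value that drives a neon curve-atlas color ramp.
# A star texture's green channel and smoothstep grid seams shape the opacity, while a dynamic Erode parameter dissolves tiles by noise and particle color tints and fades them.
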Material /BRCosmetics/Weapons/FORT_Melee/Pickaxe_SparkleChop/FX/M_Sparklechop_DiscoFloor
graph {
 properties {
  bAutomaticallySetUsageInEditor=false
  bUsedWithNiagaraMeshParticles=true
  bUsedWithNiagaraSprites=true
  blendMode=BLEND_Translucent
  shadingModel=MSM_Unlit
  usedShadingModels=Unlit
 }
 nodes [
  n1 = TextureCoordinate()
  n2 = ScalarParameter() {defaultValue=12, parameterName=Tiling}
  n3 = Multiply(A=@n1, B=@n2)
  n4 = Time()
  n5 = ScalarParameter() {defaultValue=1, parameterName=TimeMultiplier}
  n6 = ScalarParameter() {defaultValue=6, parameterName=StepAmount}
  n7 = Multiply(A=Frac(None=Multiply(A=@n4, B=@n5) {constB=0.5}), B=@n6)
  n8 = Add(A=Divide(A=Floor(None=@n3), B=@n2), B=Divide(A=Ceil(None=@n7), B=@n6))
  n9 = TextureSample(UVs=@n1) {texture=/BRCosmetics/Accessories/FORT_Backpacks/Backpack_F_MED_WaywardRebel/FX/T_WaywardRebel_Backpack_StarRGB}
  n10 = Multiply(A=TextureSample(UVs=Sine(None=@n8)) {texture=/Game/Global/Textures/Default/Noise/3xPerlin_M}.R, B=OneMinus(None=Step(Y=@n9.G) {constX=0.2}))
  n11 = Constant() {r=0.1}
  n12 = Frac(None=Multiply(A=@n1, B=@n2))
  n13 = Abs(None=@n12)
  n14 = ComponentMask(None=@n13) {r=true}
  n15 = ComponentMask(None=@n13) {g=true}
  n16 = OneMinus(None=Multiply(
    A=SmoothStep(Min=@n11, Max=@n11, Value=@n14) {constMax=0.45, constMin=0.45}
    B=SmoothStep(Min=@n11, Max=@n11, Value=@n15) {constMax=0.45, constMin=0.45}
   ))
  n17 = CurveAtlasRowParameter(CurveTime=Multiply(A=Subtract(A=@n10, B=@n16)) {constB=3}) {atlas=&AthenaCurveAtlas01, curve=/Game/Athena/LinearColorCurves/Curve_Neon_Ramp1, defaultValue=63}
  n18 = Multiply(A=@n17) {constB=0.5}
  n19 = Distance(A=Abs(None=Subtract(A=@n12) {constB=0.5}), B=Constant() {r=0.75})
  n20 = Multiply(A=@n9.G, B=Saturate(None=OneMinus(None=@n19)))
  n21 = ParticleColor()
  MaterialFunctionCall(Emissive=Multiply(A=LinearInterpolate(A=@n17, B=@n18, Alpha=@n20), B=@n21.RGB)) {materialFunction=&MF_FX_EmissiveAdjustments}
  n22 = Subtract(A=@n9.G, B=@n16)
  n23 = DynamicParameter() {defaultValue=<rgba 0 1 1 1>, paramNames [Erode,Param2,Param3,Param4]}
  n24 = SmoothStep(Min=@n23.Erode, Max=Add(A=@n23.Erode) {constB=0.1}, Value=@n10)
  Multiply(A=Multiply(A=Saturate(None=@n22), B=Saturate(None=@n24)), B=@n21.A)
 ]
}
