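# Niagara swing trail effect for the SparkleChop pickaxe cosmetic, pairing an anim-trail ribbon with flickering sprite sparkles.
# Maintrail builds a short-lived ribbon from anim trail bones, tinted by a time-of-day and variant color selector, with a dynamic material parameter driven by emitter loop age.
# SparklesTrail spawns by rate and distance from a sampled socket, inherits velocity, and drags to a stop.
# A randomized waveform flickers sparkle size while alpha fades, scaled separately for front end.
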NiagaraSystem /BRCosmetics/Weapons/FORT_Melee/Pickaxe_SparkleChop/FX/NS_Pickaxe_SparkleChop_Trail
dependencies {
 usedDataInterfaces [&NiagaraDataInterfaceCurve, &NiagaraDataInterfaceSocketReader]
 usedDynamicInputs [&Multiply_VectorByFloat, &RandomRangeVector]
 usedModules [
  &self:SystemSpawnScript.SystemScriptSource.SystemScriptGraph.NiagaraNodeAssignment_1.SetVariables_#1
  &SystemState
  &FN_VariantManager
 ]
}
emitters {
 Maintrail {
  emitterUpdateScript [
   &EmitterState()
   &FN_SpawnAnimTrail()
  ]
  particleSpawnScript [
   &FN_InitializeAnimTrailBoneIndecies()
   &FN_InitializeAnimTrail(
    AnimTrailWidth=0.5
    Color=&VariantColorSelector(
     Variant Amount=0
     Morning Color=<rgba 0.8196 0.4824 1 1>
     Day Color=<rgba 1 1 1 1>
     Evening Color=<rgba 0.974 0.4513 0.2549 1>
     Night Color=<rgba 0.2 0.251 0.502 1>
     Use TOD?=true
     Original Color=<rgba 1 1 1 0.65>
     Variant 1 Color=<rgba 0.5855 0.0715 1 0.65>
     Variant 2 Color=?
    )
    Lifetime=0.25
   )
  ]
  particleUpdateScript [
   &ParticleState()
   &FN_UpdateAnimTrail()
   &DynamicMaterialParameters(
    Index 0 Param 1=&FloatFromCurve(
     FloatCurve=DataInterface() {propertyValues {Curve {keys [{tangentMode=RCTM_User, time=0, value=0}, {tangentMode=RCTM_User, time=1, value=1}]}}}
     CurveIndex=$Emitter.NormalizedLoopAge
     Scale Curve=-2.5
    )
    Module.Index0Param0Write=true
    Module.Index0Param1Write=true
    Module.Index0Param2Write=true
    Module.Index0Param3Write=true
   )
  ]
  rendererClasses [&NiagaraRibbonRendererProperties]
  renderers [{rendererClass=&NiagaraRibbonRendererProperties}]
 }
 SparklesTrail {
  emitterUpdateScript [
   &EmitterState(Life Cycle Mode=::NewEnumerator1, Loop Behavior=::NewEnumerator1, Loop Delay=0.025, Loop Duration=1, UseLoopDelay=true)
   &SpawnPerUnit(Module.Use Max Movement Threshold=true, Module.Use Movement Tolerance=false, Spawn Spacing=33, Use Spawn Probability=false)
   &SpawnRate(SpawnRate=&RandomRangeFloat(Minimum=30, Maximum=40), Use Spawn Probability=false)
  ]
  particleSpawnScript [
   &InitializeParticle(
    Lifetime Max=1
    Lifetime Min=0.5
    Lifetime Mode=::NewEnumerator1
    Mesh Scale Mode=::NewEnumerator1
    Mesh Uniform Scale Max=1.2
    Mesh Uniform Scale Min=0.9
    Sprite Size Max=<33 33>
    Sprite Size Min=<17 17>
    Sprite Size Mode=::NewEnumerator2
    UsePositionOffset=false
    Write Material Random=true
   )
   &InitialMeshOrientation(Use Rotation Vector=false)
   &InheritVelocity(Module.Limit Speed=true)
   &AddVelocityFromPoint(Module.Use Falloff Distance=false, Velocity Strength=50)
   &SocketLocation(
    Interpolate Socket Sampling=true
    SamplePosOffset=true
    Socket Index=3
    Socket Reader=$System.Socket Reader
    Socket Sampling Mode=::NewEnumerator2
    VelocitySpeedLimit=false
   )
  ]
  particleUpdateScript [
   &ParticleState()
   /SetVariables(Particles.Flicker:NiagaraFloat=&Waveform(
     Global Input Position=&RandomRangeFloat(Minimum=0, Maximum=9)
     "Waveform [1]"=::NewEnumerator10
     "[1] Frequency"=9
     "[1] Amplitude Min/Max"=<0.6 1>
    )) {isSetParameters=true}
   &Drag(Drag=2, Use Linear Drag=true, Use Rotational Drag=false)
   &SolveForcesAndVelocity(Clamp Velocity=false, Limit Acceleration=false)
   &ScaleSpriteSize(
    Uniform Curve Scale=$Particles.Flicker
    Uniform Curve Sprite Scale=DataInterface() {propertyValues {Curve {keys [
        {arriveTangent=9.0725, interpMode=RCIM_Cubic, leaveTangent=9.0725, tangentMode=RCTM_User, time=0, value=0}
        {arriveTangent=-2.5249, interpMode=RCIM_Cubic, leaveTangent=-2.5249, tangentMode=RCTM_User, time=0.1408, value=0.9847}
        {arriveTangent=-0.0405, interpMode=RCIM_Cubic, leaveTangent=-0.0405, tangentMode=RCTM_User, time=1, value=0}
       ]}}}
   )
   &ScaleColor(
    Scale Alpha=&FloatFromCurve(
     FloatCurve=DataInterface() {propertyValues {Curve {keys [<at 0 1>,<at 1 0>]}}}
     Scale Curve=&FN_FrontEndFloatModifier(Gameplay=0.8, Front End=0.75)
    )
    ScaleA=true
    ScaleRGB=true
   )
  ]
  rendererClasses [&NiagaraSpriteRendererProperties]
  renderers [{rendererClass=&NiagaraSpriteRendererProperties}]
 }
}
summary {
 systemName=NS_Pickaxe_SparkleChop_Trail
 emitters [
  {emitterName=SparklesTrail, rendererClasses [&NiagaraSpriteRendererProperties]}
  {emitterName=Maintrail, rendererClasses [&NiagaraRibbonRendererProperties]}
 ]
 userVariables [
  User.bIsFrontEnd {type=NiagaraBool}
  User.bIsFrontEndPreview {type=NiagaraBool}
  User.RibbonIDOffset {type=NiagaraInt32}
  User.Variant {type=NiagaraFloat, defaultValue=0, description=""}
 ]
}
