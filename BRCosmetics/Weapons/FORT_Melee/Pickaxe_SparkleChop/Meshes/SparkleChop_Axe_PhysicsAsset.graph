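# Physics asset for the SparkleChop pickaxe mesh, a Battle Royale cosmetic two-handed melee weapon.
# It has a single body on the Melee_TwoHanded_Handle bone, made of one large capsule that is offset and rotated from the bone origin.
# It defines no constraints, so the weapon has one collision volume with no jointed parts.
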PhysicsAsset /BRCosmetics/Weapons/FORT_Melee/Pickaxe_SparkleChop/Meshes/SparkleChop_Axe_PhysicsAsset
bodies {Melee_TwoHanded_Handle [{shapeType=Capsule, length=74.6211, radius=51.8363, center=<0.8788 -2.1948 28.6662>, rotation=<rot 1.9782 -90 0>}]}
constraints []
